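; Base weakpoint actor, attached to a parent actor, that tracks its own damageable health.
; Native health is set huge and healed by each hit's damage, while authority-side internal health drives disabling and optional destruction at zero.
; Hits notify an overridable parent hook and optionally re-apply damage to the parent through a GameplayEffect.
; Disabled weakpoints zero incoming damage with ricochet effects, and can switch off collision through a RepNotify.
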(event EventBeginPlay
  (Utilities|FlowControl|Sequence
    (:then_0
      ;[0.01586 0.2656 0.157 0.5] Initialize Weakpoint Internal Data
      (Health|SetMaxHealthDontCreateASC 999999999.0)
      (CallFunction|ResetWeakpointHealth))))

(event EventOnDamageServer
  (params (Damage :type Float%20%28single-precision%29) (DamageTags :type Gameplay%20Tag%20Container%20Structure%20%28by%20ref%29) (Momentum :type Vector :default "0, 0, 0") (HitInfo :type Hit%20Result%20Structure%20%28by%20ref%29) (InstigatedBy :type Controller%20Object%20Reference) (DamageCauser :type Actor%20Object%20Reference) (EffectContext :type Gameplay%20Effect%20Context%20Handle%20Structure))
  (bind _self self)
  (bind _returnvalue (Ability|GetFortAbilitySystemComponent _self))
  (|CallParentFunction Damage DamageTags Momentum HitInfo InstigatedBy DamageCauser EffectContext)
  (bind _returnvalue_2 (Ability|GameplayEffect|AssignTagSetbyCallerMagnitude (GameplayEffects|MakeOutgoingSpec :self _returnvalue :GameplayEffectClass "/ActorWeakpointsCore/GameplayEffects/GE_Weakpoint_InstantHeal.GE_Weakpoint_InstantHeal_C" :Level 1.0) "(TagName=\"SetByCaller.Healing\")" Damage))
  (GameplayEffects|ApplyGameplayEffectSpecToTarget _returnvalue _returnvalue_2 _returnvalue)
  (CallFunction|DamageWeakpoint Damage)
  (CallFunction|NotifyParent Damage InstigatedBy DamageCauser)
  (if (Variables|Tunable|GetbDamageParentOnWeakpointDamaged)
    (bind _returnvalue_3 (Spec|MakeSpecHandle(ByClass) "/ActorWeakpointsCore/GameplayEffects/GE_Weakpoint_DamageParent.GE_Weakpoint_DamageParent_C" InstigatedBy self))
    (bind _returnvalue_4 (Ability|GameplayEffect|AssignTagSetbyCallerMagnitude _returnvalue_3 "(TagName=\"SetByCaller.Damage\")" Damage))
    (GameplayEffects|ApplyGameplayEffectSpecToTarget (Ability|GetAbilitySystemComponent InstigatedBy) _returnvalue_4 (Ability|GetAbilitySystemComponent (Actor|GetParentActor self)))))

(event EventOnDamagePlayEffects
  (params (Damage :type Float%20%28single-precision%29) (DamageTags :type Gameplay%20Tag%20Container%20Structure%20%28by%20ref%29) (Momentum :type Vector :default "0, 0, 0") (HitInfo :type Hit%20Result%20Structure%20%28by%20ref%29) (InstigatedBy :type Fort%20Pawn%20Object%20Reference) (DamageCauser :type Actor%20Object%20Reference) (EffectContext :type Gameplay%20Effect%20Context%20Handle%20Structure))
  (bind (_bblockinghit _binitialoverlap _time _distance _location _impactpoint _normal) (Collision|BreakHitResult HitInfo))
  (bind _weakpoint_hit_niagarasystem (Variables|Tunable|GetWeakpoint_Hit_NiagaraSystem))
  ;[0.9271 0.1015 0.09266 0.5] Play Hit FX
  (Variables|Internal|SetInstigatorPawn InstigatedBy)
  (if (Variables|Internal|GetbIsEnabled)
    (Utilities|FlowControl|Sequence
      (:then_0
        (Utilities|IsValid _weakpoint_hit_niagarasystem
          (:Is%20Valid
            (Networking|CanExecuteCosmeticEvents
              (:True
                (bind _returnvalue (Effects|Components|ParticleSystem|SpawnFXAtLocation _weakpoint_hit_niagarasystem _location (Math|Conversions|RotationFromXVector _normal)))
                (Effects|Components|ParticleSystem|SetVectorParameter _returnvalue "ImpactFXDirection" _normal))))))
      (:then_1
        (Utilities|IsValid (Variables|Tunable|GetWeakpoint_Hit_SoundFX)
          (:Is%20Valid
            (if (Pawn|IsLocallyControlled InstigatedBy)
              (Audio|PlaySoundatLocation (Variables|Tunable|GetWeakpoint_Hit_SoundFX) (Transformation|GetActorLocation InstigatedBy) "0, 0, 0" 1.3))))))))

(event Custom|PlayRicochet
  (params (Hit%20Result :type Hit%20Result%20Structure))
  (bind (_bblockinghit _binitialoverlap _time _distance _location _impactpoint _normal) (Collision|BreakHitResult Hit%20Result))
  (bind _weakpoint_ricochetfx (Variables|Tunable|GetWeakpoint_RicochetFX))
  ;[1 0.2545 0.03082] Play Ricochet Hit VFX
  (Utilities|IsValid _weakpoint_ricochetfx
    (:Is%20Valid
      (Networking|CanExecuteCosmeticEvents
        (:True
          (bind _returnvalue (Effects|Components|ParticleSystem|SpawnFXAtLocation _weakpoint_ricochetfx _location (Math|Conversions|RotationFromXVector _normal)))
          (Effects|Components|ParticleSystem|SetVectorParameter _returnvalue "ImpactFXDirection" _normal)
          (Variables|Tunable|GetWeakpoint_Ricochet_SoundFX
            (:then
              (bind _instigatorpawn (Variables|Internal|GetInstigatorPawn)
                (:then
                  (if (Pawn|IsLocallyViewed _instigatorpawn)
                    (Audio|PlaySoundatLocation (Variables|Tunable|GetWeakpoint_Ricochet_SoundFX) _location)))
                (:else)))
            (:else)))))))

(fn BlueprintModifyIncomingDamage
  (params (Damage :type Float%20%28single-precision%29) (DamageTags :type Gameplay%20Tag%20Container%20Structure%20%28by%20ref%29) (EffectContext :type Gameplay%20Effect%20Context%20Handle%20Structure%20%28by%20ref%29) (EventInstigator :type Controller%20Object%20Reference) (DamageCauser :type Actor%20Object%20Reference))
  (|CallParentFunction Damage DamageTags EffectContext EventInstigator DamageCauser)
  (bind _modifieddamage (CallFunction|ProcessHitResult Damage (Ability|EffectContext|GetHitResult EffectContext)))
  (return _modifieddamage))

(fn DamageWeakpoint
  (params (Damage :type Float%20%28double-precision%29))
  (Networking|SwitchHasAuthority
    (:Authority
      (if (CallFunction|ShouldDamageWeakpoint)
        (Development|PrintString (Utilities|String|Append "@@@ DamageWeakpoint [BGA_WeakpointActor_Base]: CALLED for " (Utilities|GetDisplayName self) "!") false (Variables|Debug|GetbDebugWeakpoint))
        (bind _output_get (Variables|Internal|SetWeakpoint_CurrentHealth (Math|Float|SelectFloat 0.0 (- (Variables|Internal|GetWeakpoint_CurrentHealth) (Variables|GetDamage)) (< (- (Variables|Internal|GetWeakpoint_CurrentHealth) (Variables|GetDamage)) 0))))
        (Development|PrintString (Utilities|String|Append "@@@@@@ DamageWeakpoint [BGA_WeakpointActor_Base]: current health = " (Utilities|String|ToString(Float) _output_get) "!") false (Variables|Debug|GetbDebugWeakpoint))
        (if (<= _output_get 0)
          (CallFunction|ProcessHealthDepleted))))))

(fn DestroyWeakpoint
  (bind _self self)
  (bind _returnvalue (Actor|GetParentActor _self))
  (bind _returnvalue_2 (Actor|GetComponentbyClass _returnvalue "/ActorWeakpointsCore/Components/BP_ActorWeakpoints_UtilityComponent.BP_ActorWeakpoints_UtilityComponent_C"))
  (Networking|SwitchHasAuthority
    (:Authority
      (Development|PrintString (Utilities|String|Append "@@@ DestroyWeakpoint [BGA_WeakpointActor_Base]: CALLED for " (Utilities|GetDisplayName self) "!") false (Variables|Debug|GetbDebugWeakpoint))
      (Utilities|IsValid _returnvalue_2
        (:Is%20Valid
          (Class|BPActorWeakpointsUtilityComponent|RemoveWeakpoint _returnvalue_2 self))
        (:Is%20Not%20Valid
          _))
      (Variables|Internal|SetbCollisionEnabled false)
      (Actor|DestroyActor)
      (Default|CallWeakPointDestroyed))))

(fn DisableWeakpoint
  (Networking|SwitchHasAuthority
    (:Authority
      (if (Variables|Internal|GetbIsEnabled)
        (Development|PrintString (Utilities|String|Append "@@@ DisableWeakpoint [BGA_WeakpointActor_Base]: CALLED for " (Utilities|GetDisplayName self) "!") false (Variables|Debug|GetbDebugWeakpoint))
        (Variables|Internal|SetbIsEnabled false)
        (if (Variables|Tunable|GetbDisableCollisionWhenWeakpointIsDisabled)
          (Variables|Internal|SetbCollisionEnabled false))))))

(fn EnableWeakpoint
  (Networking|SwitchHasAuthority
    (:Authority
      (if (not (Variables|Internal|GetbIsEnabled))
        (Development|PrintString (Utilities|String|Append "@@@ EnableWeakpoint [BGA_WeakpointActor_Base]: CALLED for " (Utilities|GetDisplayName self) "!") false (Variables|Debug|GetbDebugWeakpoint))
        (Variables|Internal|SetbIsEnabled true)
        (if (Variables|Tunable|GetbDisableCollisionWhenWeakpointIsDisabled)
          (Variables|Internal|SetbCollisionEnabled true))))))

(fn GetTargetOwner
  (return 0))

(fn HandleCollisionEnabled
  (CallFunction|UpdateWeakpointCollision (Variables|Internal|GetbCollisionEnabled) (Variables|Internal|GetWeakpointCollisionKind)))

(fn IsValidAutoFireTarget
  (params (Attacker :type Actor%20Object%20Reference))
  (return false))

(fn NotifyParent
  (params (Damage :type Float%20%28double-precision%29) (InstigatedBy :type Controller%20Object%20Reference) (DamageCauser :type Actor%20Object%20Reference)))

(fn OnRep_bCollisionEnabled
  (CallFunction|HandleCollisionEnabled))

(fn ProcessHealthDepleted
  (CallFunction|DisableWeakpoint)
  (if (Variables|Tunable|GetbDestroyWeakpointOnHealthDepleted)
    (CallFunction|DestroyWeakpoint)))

(fn ProcessHitResult
  (params (Damage :type Float%20%28double-precision%29) (HitResult :type Hit%20Result%20Structure))
  (if (Variables|Internal|GetbIsEnabled)
    (return Damage)
    (else
      (CallFunction|PlayRicochet HitResult)
      (return 0.0))))

(fn ResetWeakpointHealth
  (bind _weakpoint_maxhealth (Variables|Tunable|GetWeakpoint_MaxHealth))
  (bind _returnvalue (ScalableFloat|GetValueatLevel _weakpoint_maxhealth))
  (Development|PrintString (Utilities|String|Append "@@@ ResetWeakpointHealth [BGA_WeakpointActor_Base]: CALLED for " (Utilities|GetDisplayName self) "!") false (Variables|Debug|GetbDebugWeakpoint))
  (if (> _returnvalue 0)
    (Variables|Internal|SetWeakpoint_CurrentHealth _returnvalue)))

(fn SetWeakpointCollisionKind
  (params (WeakpointCollisionKind :type Enum_WeakpointActor_CollisionKind%20Enum :default "NewEnumerator0") (BoxCollision :type Box%20Collision%20Object%20Reference) (CapsuleCollision :type Capsule%20Collision%20Object%20Reference) (SphereCollision :type Sphere%20Collision%20Object%20Reference))
  (bind _output_get (Variables|Internal|SetWeakpointCollisionKind WeakpointCollisionKind))
  (switch Utilities|FlowControl|Switch|SwitchonEnum_WeakpointActor_CollisionKind _output_get
    (:NewEnumerator0
      (Variables|Internal|SetWeakpointCollision_Box (Variables|GetBoxCollision)))
    (:NewEnumerator1
      (Variables|Internal|SetWeakpointCollision_Capsule (Variables|GetCapsuleCollision)))
    (:NewEnumerator2
      (Variables|Internal|SetWeakpointCollision_Sphere (Variables|GetSphereCollision)))
    (:NewEnumerator3
      (Variables|Internal|SetWeakpointCollision_Box 0)
      (Variables|Internal|SetWeakpointCollision_Capsule 0)
      (Variables|Internal|SetWeakpointCollision_Sphere 0))))

(fn ShouldAlliesAutoFireOnDeployedTarget
  (return false))

(fn ShouldAutoFireOnSelfDeployedTarget
  (return false))

(fn ShouldDamageWeakpoint
  (return (and (Variables|Internal|GetbIsEnabled) (Variables|Tunable|GetbWeakpointCanBeDamaged))))

(fn UpdateWeakpointCollision
  (params (bEnableCollision :type Boolean) (CollisionKind :type Enum_WeakpointActor_CollisionKind%20Enum :default "NewEnumerator0") (CollisionComponent :type Primitive%20Component%20Object%20Reference))
  (bind _returnvalue (Utilities|IsValid CollisionComponent :!class /Script/Engine.KismetSystemLibrary))
  (bind _collisioncomponent (Variables|GetCollisionComponent))
  (bind _collisionkindcomponent (LocalVariables|GetCollisionKindComponent))
  (bind _returnvalue_2 (select _returnvalue _collisioncomponent _collisionkindcomponent))
  (Development|PrintString (Utilities|String|Append "@@@@@@ UpdateWeakpointCollision [BGA_WeakpointActor_Base]: CALLED for " (Utilities|GetDisplayName self) "!") false (Variables|Debug|GetbDebugWeakpoint))
  (Utilities|IsValid CollisionComponent
    (:Is%20Valid
      _)
    (:Is%20Not%20Valid
      (switch Utilities|FlowControl|Switch|SwitchonEnum_WeakpointActor_CollisionKind CollisionKind
        (:NewEnumerator0
          (LocalVariables|SetCollisionKindComponent (Variables|Internal|GetWeakpointCollision_Box)))
        (:NewEnumerator1
          (LocalVariables|SetCollisionKindComponent (Variables|Internal|GetWeakpointCollision_Capsule)))
        (:NewEnumerator2
          (LocalVariables|SetCollisionKindComponent (Variables|Internal|GetWeakpointCollision_Sphere))))))
  (Collision|SetCollisionResponsetoChannel _returnvalue_2 "ECC_GameTraceChannel2" (select (Variables|GetbEnableCollision) "ECR_Block" "ECR_Ignore"))
  (Collision|SetCollisionResponsetoChannel _returnvalue_2 "ECC_GameTraceChannel3" (select (Variables|GetbEnableCollision) "ECR_Block" "ECR_Ignore"))
  (Collision|SetCollisionResponsetoChannel _returnvalue_2 "ECC_GameTraceChannel3" (select (Variables|GetbEnableCollision) "ECR_Block" "ECR_Ignore"))
  (Collision|SetCollisionResponsetoChannel _returnvalue_2 "ECC_GameTraceChannel5" (select (Variables|GetbEnableCollision) "ECR_Block" "ECR_Ignore"))
  (Collision|SetCollisionResponsetoChannel _returnvalue_2 "ECC_GameTraceChannel6" (select (Variables|GetbEnableCollision) "ECR_Block" "ECR_Ignore"))
  (Collision|SetCollisionResponsetoChannel _returnvalue_2 "ECC_GameTraceChannel7" (select (Variables|GetbEnableCollision) "ECR_Block" "ECR_Ignore"))
  (Collision|SetCollisionResponsetoChannel _returnvalue_2 "ECC_GameTraceChannel10" (select (Variables|GetbEnableCollision) "ECR_Block" "ECR_Ignore"))
  (Collision|SetCollisionResponsetoChannel _returnvalue_2 "ECC_GameTraceChannel11" (select (Variables|GetbEnableCollision) "ECR_Block" "ECR_Ignore"))
  (Collision|SetCollisionResponsetoChannel _returnvalue_2 "ECC_GameTraceChannel14" (select (Variables|GetbEnableCollision) "ECR_Overlap" "ECR_Ignore"))
  (Collision|SetCollisionResponsetoChannel _returnvalue_2 "ECC_GameTraceChannel8" (select (Variables|GetbEnableCollision) "ECR_Block" "ECR_Ignore"))
  (Collision|SetCollisionResponsetoChannel _returnvalue_2 "ECC_GameTraceChannel9" (select (Variables|GetbEnableCollision) "ECR_Block" "ECR_Ignore"))
  (Collision|SetCollisionResponsetoChannel _returnvalue_2 "ECC_GameTraceChannel12" (select (Variables|GetbEnableCollision) "ECR_Block" "ECR_Ignore"))
  (Collision|SetCollisionResponsetoChannel _returnvalue_2 "ECC_WorldStatic" (select (Variables|GetbEnableCollision) "ECR_Overlap" "ECR_Ignore"))
  (Collision|SetCollisionResponsetoChannel _returnvalue_2 "ECC_WorldDynamic" (select (Variables|GetbEnableCollision) "ECR_Overlap" "ECR_Ignore"))
  (Collision|SetCollisionResponsetoChannel _returnvalue_2 "ECC_Pawn" (select (Variables|GetbEnableCollision) "ECR_Overlap" "ECR_Ignore"))
  (Collision|SetCollisionResponsetoChannel _returnvalue_2 "ECC_PhysicsBody" (select (Variables|GetbEnableCollision) "ECR_Overlap" "ECR_Ignore"))
  (Collision|SetCollisionResponsetoChannel _returnvalue_2 "ECC_Vehicle" (select (Variables|GetbEnableCollision) "ECR_Overlap" "ECR_Ignore"))
  (Collision|SetCollisionResponsetoChannel _returnvalue_2 "ECC_Destructible" (select (Variables|GetbEnableCollision) "ECR_Overlap" "ECR_Ignore"))
  (Collision|SetCollisionResponsetoChannel _returnvalue_2 "ECC_GameTraceChannel4" (select (Variables|GetbEnableCollision) "ECR_Overlap" "ECR_Ignore"))
  (Collision|SetCollisionEnabled _returnvalue_2 (select (Variables|GetbEnableCollision) "QueryAndPhysics" "NoCollision")))

(fn ConstructionScript)
